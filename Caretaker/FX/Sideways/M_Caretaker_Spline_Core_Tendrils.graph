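# Unlit translucent, two-sided material for Caretaker tendril effects on Niagara ribbons and mesh particles.
# Two panning noise textures multiply into a shared mask, alongside shadow-to-light color blending tinted by particle color.
# A displaced sphere mask, offset by dynamic AnimPan, combines with a square edge mask and particle alpha for opacity.
# World position offset uses time-driven sine or cosine waves, chosen per particle and scaled by dynamic parameter X.
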Material /Caretaker/FX/Sideways/M_Caretaker_Spline_Core_Tendrils
graph {
 properties {
  bUsedWithNiagaraMeshParticles=true
  bUsedWithNiagaraRibbons=true
  blendMode=BLEND_Translucent
  pixelDepthOffsetMode=PDOM_Legacy
  shadingModel=MSM_Unlit
  twoSided=true
  usedShadingModels=Unlit
 }
 nodes [
  n1 = ScalarParameter() {defaultValue=1.5, group=SphereMask, parameterName=EdgeHardness}
  n2 = NamedRerouteUsage()
  LinearInterpolate(
   B=@n1
   Alpha=CurveAtlasRowParameter(CurveTime=MaterialFunctionCall(Float2=@n2) {materialFunction=&BreakOutFloat2Components}.G) {atlas=/Caretaker/FX/Curves/CurveAtlas_Caretaker, curve=/Caretaker/FX/Curves/Curve_Caretaker_Core01, parameterName=ShadowCurve}.R
  ) {constB=0}
  n3 = NamedRerouteUsage()
  n4 = ParticleColor()
  n5 = Multiply(
   A=LinearInterpolate(
    A=VectorParameter() {parameterName=ShadowColor, defaultValue=<rgba 0.005 0.0008 0.0028 1>}.RGB
    B=VectorParameter() {parameterName=LightColor, defaultValue=<rgba 0.3038 0.1523 0.651 1>}.RGB
    Alpha=LinearInterpolate(Alpha=@n3)
   )
   B=@n4.RGB
  )
  MaterialFunctionCall(Emissive=@n5) {materialFunction=&MF_FX_EmissiveAdjustments}
  NamedRerouteDeclaration(None=@n5) {name=Color, nodeColor=<rgba 0.8234 0.0183 1 1>}
  n6 = NamedRerouteUsage()
  n7 = Constant2Vector() {g=1.8, r=1}
  n8 = Add(A=@n6, B=LinearInterpolate(Alpha=NamedRerouteUsage()) {constA=-0.5, constB=0.5})
  n9 = ScalarParameter() {defaultValue=2, parameterName=MaskNoiseDisplacement}
  n10 = MaterialFunctionCall(Float2=Add(
    A=MaterialFunctionCall(UVs=@n6, Texture Scale=@n7) {materialFunction=&ScaleUVsByCenter}.UVs
    B=Divide(A=@n8, B=@n9) {constB=3}
   )) {materialFunction=&BreakOutFloat2Components}
  n11 = DynamicParameter() {defaultValue=<rgba 1 1 1 0>, paramNames [X,Y,Z,AnimPan]}
  n12 = SphereMask(
   A=AppendVector(A=@n10.R, B=Add(A=@n10.G, B=@n11.AnimPan))
   B=ComponentMask(None=VectorParameter() {group=SphereMask, parameterName=SphereMaskPos, defaultValue=<rgba 0.5 0.5 0 0>}.RGB) {g=true, r=true}
   Radius=ScalarParameter() {defaultValue=0.4, group=SphereMask, parameterName=SphereMaskRaidus}
   Hardness=@n1
  )
  n13 = MaterialFunctionCall!(UVs, Mult, Exp=LinearInterpolate(Alpha=NamedRerouteUsage()) {constA=0.5, constB=2}) {materialFunction=/Game/Effects/Fort_Effects/MaterialFunctions/MF_SquareEdgeMask}
  NamedRerouteDeclaration(None=Multiply(A=@n4.A, B=Multiply(A=@n12, B=@n13.SquareMask))) {name=Opacity, nodeColor=<rgba 0 1 0.5078 1>}
  n14 = DynamicParameter() {defaultValue=<rgba 1 1 1 0>, paramNames [X,Y,Z,AnimPan]}
  n15 = MaterialFunctionCall(R=@n14.X, G=@n14.X, B=@n14.X) {materialFunction=&AppendMany}
  n16 = NamedRerouteUsage()
  n17 = ScalarParameter() {defaultValue=1.8, parameterName=WPO_Tile}
  n18 = Divide(A=Time()) {constB=6}
  n19 = Multiply(A=ParticleRandom()) {constB=3}
  n20 = MaterialFunctionCall(Float2=Add(A=Multiply(A=@n16, B=@n17), B=Add(A=@n18, B=@n19))) {materialFunction=&BreakOutFloat2Components}
  n21 = LinearInterpolate(A=Sine(None=@n20.G), B=Cosine(None=@n20.G), Alpha=Round(None=ParticleRandom()))
  NamedRerouteDeclaration(None=Multiply(A=Multiply(A=@n15.RGB, B=@n21), B=ScalarParameter() {defaultValue=40, parameterName=WPO_Intensity})) {name=WPO, nodeColor=<rgba 0 0.7656 1 1>}
  n22 = TextureCoordinate()
  NamedRerouteDeclaration(None=LinearInterpolate(
    A=@n22
    B=MaterialFunctionCall(XY=@n22) {materialFunction=&Swizzle}.YX
    Alpha=ScalarParameter() {parameterName=SwizzleUV}
   )) {name=TextureCoord, nodeColor=<rgba 1 0 0.0343 1>}
  n23 = NamedRerouteUsage()
  n24 = ComponentMask(None=VectorParameter() {parameterName=MaskNoiseTile, defaultValue=<rgba 1 7 0 0>}.RGB) {g=true, r=true}
  n25 = Multiply(A=@n24) {constB=0.7}
  n26 = VectorParameter() {parameterName=MaskNoiseA_PannerSpeed, defaultValue=<rgba 0.2 0.3 0 0>}
  n27 = TextureSampleParameter2D(UVs=Panner(Coordinate=Multiply(A=@n23, B=@n25), Speed=ComponentMask(None=@n26.RGB) {g=true, r=true}) {speedX=0.2, speedY=0.3}) {parameterName=NoiseA, samplerType=SAMPLERTYPE_LinearGrayscale, texture=/Game/Environments/Hexmap/HexClouds/T_HexmapFog_01_Height02}
  n28 = Multiply(A=@n24) {constB=0.5}
  n29 = NamedRerouteUsage()
  n30 = VectorParameter() {parameterName=MaskNoiseB_PannerSpeed, defaultValue=<rgba 0.1 0.7 0 0>}
  n31 = TextureSampleParameter2D(UVs=Panner(Coordinate=Multiply(A=@n28, B=@n29), Speed=ComponentMask(None=@n30.RGB) {g=true, r=true}) {speedX=0.1, speedY=0.7}) {
   parameterName=NoiseB
   samplerType=SAMPLERTYPE_LinearGrayscale
   texture=/Game/Global/Textures/Default/Tileable_Patterns/T_StormCloudDetail01_M
  }
  NamedRerouteDeclaration(None=Multiply(A=@n27.R, B=@n31.R)) {name=NoiseInput, nodeColor=<rgba 1 0 0.3984 1>}
  NamedRerouteUsage()
  NamedRerouteUsage()
  NamedRerouteUsage()
 ]
}
