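# Translucent, two-sided Niagara particle material for the Sideways portal opening effect, drawing a glowing, noisy ring.
# Sphere masks on noise-distorted UV1 shape the ring, with dynamic parameter InnerRadius controlling the hole.
# Blurred noise combines with polar-mapped caustics and layered, bump-offset swirl membranes inside that mask.
# A three-color Fresnel and depth-fade gradient tints the emissive; particle color and depth fade modulate the result.
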Material /Caretaker/FX/Sideways/M_Sideways_Portal_Opening
graph {
 properties {
  bUsedWithNiagaraMeshParticles=true
  bUsedWithNiagaraSprites=true
  blendMode=BLEND_Translucent
  pixelDepthOffsetMode=PDOM_Legacy
  twoSided=true
  usedShadingModels=Default Lit
 }
 nodes [
  n1 = TextureCoordinate()
  n2 = AppendVector(
   A=ScalarParameter() {defaultValue=16, parameterName=BubblePattern_X_Tiling}
   B=ScalarParameter() {defaultValue=8, parameterName=BubblePattern_Y_Tiling}
  )
  n3 = Time()
  n4 = AppendVector(A=ScalarParameter() {parameterName=X_NoisePanSpeed}, B=ScalarParameter() {defaultValue=0.03, parameterName=Y_NoisePanSpeed})
  n5 = TextureSample(UVs=Add(A=Multiply(A=@n1, B=@n2), B=Multiply(A=@n3, B=@n4))) {texture=/Caretaker/FX/Sideways/Textures/T_Caustics_Pattern1}
  n6 = TextureSampleParameter2D(UVs=Panner() {speedY=0.25}) {parameterName=SmokeyNoise, texture=&T_NoisePack_01}
  Add(A=Clamp(None=Subtract(A=@n5.G, B=@n6.G)))
  ComponentMask(None=Power(Exp=Constant() {r=0.15})) {g=true}
  n7 = Max(A=Fresnel(), B=LinearInterpolate(A=Constant() {r=1}, B=Constant(), Alpha=DepthFade() {fadeDistanceDefault=800}))
  n8 = Max(A=Fresnel(), B=LinearInterpolate(A=Constant() {r=1}, B=Constant(), Alpha=DepthFade() {fadeDistanceDefault=20000}))
  n9 = LinearInterpolate(
   A=VectorParameter() {parameterName=Linear_Color_01, defaultValue=<rgba 1 0.3419 0.0144 1>}.RGB
   B=LinearInterpolate(
    A=VectorParameter() {parameterName=Linear_Color_02, defaultValue=<rgba 0.0545 0.0185 0.1356 1>}.RGB
    B=VectorParameter() {parameterName=Linear_Color_03, defaultValue=<rgba 0.349 0 0.1652 1>}.RGB
    Alpha=Clamp(None=@n7)
   ) {constAlpha=0.35}
   Alpha=Clamp(None=Power(Base=@n8) {constExponent=0.2})
  )
  n10 = Panner(Coordinate=TextureCoordinate() {coordinateIndex=1}, Speed=Constant() {r=0.2})
  n11 = Subtract(A=Multiply(A=TextureSampleParameter2D(UVs=@n10) {parameterName=DistortionNoise, texture=&T_NoisePack_01}.RGB) {constB=2})
  n12 = Add(
   A=TextureCoordinate() {coordinateIndex=1}
   B=Multiply(A=ComponentMask(None=@n11) {g=true, r=true}, B=ScalarParameter() {defaultValue=0.05, parameterName=NoiseDistortion_Strength})
  )
  n13 = TextureSample(UVs=@n12) {samplerType=SAMPLERTYPE_Grayscale, texture=&LowResBlurredNoise}
  n14 = Multiply(A=@n13.RGB, B=@n13.RGB)
  n15 = Constant() {r=0.15}
  n16 = Constant() {r=0.5}
  n17 = Constant() {r=0.5}
  n18 = DynamicParameter() {defaultValue=<rgba 0 1 1 1>, paramNames [InnerRadius,Param2,Param3,Param4]}
  n19 = Multiply(
   A=SphereMask(A=@n12, B=Constant() {r=0.5}, Radius=Constant() {r=0.45}, Hardness=Constant() {r=0.5})
   B=OneMinus(None=SphereMask(A=@n12, B=@n16, Radius=@n18.InnerRadius, Hardness=@n17))
  )
  n20 = Multiply(A=ComponentMask(None=Power(Base=@n14, Exp=@n15)) {g=true}, B=@n19)
  n21 = Subtract(A=TextureCoordinate() {coordinateIndex=1}, B=Constant() {r=0.5})
  n22 = ComponentMask(None=@n21) {g=true}
  n23 = ComponentMask(None=@n21) {r=true}
  n24 = Divide(A=ComponentMask(None=Arctangent2!(Y=@n22, X=@n23)) {r=true}, B=Constant() {r=6.283})
  n25 = Time()
  n26 = Multiply(A=Add(A=Frac(None=@n24), B=Multiply(A=@n25) {constB=0}), B=ScalarParameter() {defaultValue=1, parameterName=TexCausticTileX}) {constB=0, desc=Tiling X}
  n27 = Multiply(A=@n21, B=@n21)
  n28 = Add(A=ComponentMask(None=@n27) {r=true}, B=ComponentMask(None=@n27) {g=true})
  n29 = Multiply(A=Add(A=SquareRoot(None=@n28), B=Multiply(A=@n25) {constB=0.5}), B=ScalarParameter() {defaultValue=1, parameterName=TexCausticTileY}) {constB=0, desc=Tiling Y}
  n30 = TextureSample(UVs=BumpOffset(Coordinate=AppendVector(A=@n26, B=@n29), Height=Constant() {r=1}, HeightRatioInput=Constant() {r=-0.5})) {texture=&T_ShieldCaustics}
  n31 = TextureSampleParameter2D(UVs=Panner(Coordinate=TextureCoordinate() {coordinateIndex=1}, Speed=Constant() {r=-0.25})) {parameterName=DistortionNoise, texture=&T_NoisePack_01}
  n32 = ComponentMask(None=Subtract(A=Multiply(A=@n31.RGB) {constB=2})) {g=true, r=true}
  n33 = ScalarParameter() {defaultValue=0.1, parameterName=MembraneDistortion_Strength}
  n34 = VectorParameter() {parameterName=MembraneUVScale, defaultValue=<rgba 1.5 0.75 0 0>}
  n35 = Multiply(A=Add(A=TextureCoordinate() {coordinateIndex=1}, B=Multiply(A=@n32, B=@n33)), B=AppendVector(A=@n34.R, B=@n34.R)) {constB=0}
  n36 = TextureSampleParameter2D!(UVs=@n35, Bias, Apply View MipBias) {
   mipValueMode=TMVM_MipBias
   parameterName=Membrane_Texture01
   samplerType=SAMPLERTYPE_LinearColor
   texture=/Game/Global/Textures/Default/Tileable_Patterns/T_Corruption_SwirlPack
  }
  n37 = Add(A=@n35) {constB=0.25}
  n38 = TextureSampleParameter2D(UVs=BumpOffset(
    Coordinate=Multiply(A=@n37) {constB=1.5}
    Height=Constant()
    HeightRatioInput=ScalarParameter() {defaultValue=-0.1, parameterName=HeightRatio}
   )) {
   mipValueMode=TMVM_MipBias
   parameterName=Membrane_Texture01
   samplerType=SAMPLERTYPE_LinearColor
   texture=&T_Corruption_SwirlPack
  }
  n39 = Add(A=OneMinus(None=@n36.G), B=OneMinus(None=@n38.G))
  n40 = Multiply(A=@n19, B=Power(Base=Multiply(A=@n30.RGB, B=@n39)) {constExponent=0.75})
  n41 = Add(A=@n20, B=Multiply(A=@n40, B=VectorParameter() {parameterName=ColorVeins, defaultValue=<rgba 0 0 0 1>}.RGB))
  n42 = ParticleColor()
  MaterialFunctionCall(Emissive=Multiply(A=Multiply(A=@n9, B=@n41), B=@n42.RGB)) {materialFunction=&MF_FX_EmissiveAdjustments}
  Multiply(
   A=Multiply(A=@n42.A, B=Add(A=@n20, B=@n40))
   B=DepthFade(FadeDistance=ScalarParameter() {defaultValue=200, parameterName=DepthFadeDistance}) {fadeDistanceDefault=20000}
  )
  OneMinus()
  Power(Base=Multiply(A=@n5.G) {constB=5})
 ]
}
