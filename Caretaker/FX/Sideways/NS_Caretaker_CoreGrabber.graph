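# Niagara effect for the Caretaker's core grabber, combining bright drifting embers, wavering ribbon tendrils and a reddish point light.
# Embers_Dark spawns GPU sprites in randomly aimed cones, with curl noise, drag and upward acceleration.
# Tendrils and the light-only Idle_FlareAndShard share a loop-age curve shape that kicks out along User.RandomPosOffset and recoils.
# User.GlobalScale drives ember cone length and light size, while User.CorePlusingScalar feeds a lerp with identical endpoints.
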NiagaraSystem /Caretaker/FX/Sideways/NS_Caretaker_CoreGrabber
dependencies {
 usedDataInterfaces [&NiagaraDataInterfaceCurve, &NiagaraDataInterfaceMeshRendererInfo]
 usedDynamicInputs []
 usedModules [&SystemState]
}
emitters {
 Embers_Dark {
  simTarget=GPUComputeSim
  emitterUpdateScript [
   &EmitterState(
    Enable Distance Culling=true
    Life Cycle Mode=::NewEnumerator1
    Loop Behavior=::NewEnumerator1
    Loop Duration=0.8
    MaxDistance=4500
    Module.bUseMaxDistance=true
    Module.bUseMinDistance=false
    Scalability Mode=::NewEnumerator1
    UseLoopDelay=false
   )
   &SpawnRate(
    SpawnRate=&Multiply_Float(
     A=60
     B=&FloatFromCurve(
      FloatCurve=DataInterface() {propertyValues {Curve {keys [
          <at 0 1 RCIM_Cubic>
          {arriveTangent=-0.2543, interpMode=RCIM_Cubic, leaveTangent=-0.2543, time=1.2828, value=1.6362}
          {arriveTangent=-0.3703, interpMode=RCIM_Cubic, leaveTangent=-0.3703, time=2.4472, value=0.3776}
          {arriveTangent=-0.11, interpMode=RCIM_Cubic, leaveTangent=-0.11, time=3.6302, value=0.7671}
          {arriveTangent=0.2907, interpMode=RCIM_Cubic, leaveTangent=0.2907, time=5.034, value=0.0931}
          {arriveTangent=0.3058, interpMode=RCIM_Cubic, leaveTangent=0.3058, time=6.8459, value=1.702}
          <at 8 1>
         ]}}}
      CurveIndex=$Emitter.NormalizedLoopAge
     )
    )
    Use Spawn Probability=false
   )
   /SetVariables(Emitter.ConeDir:Vector3f=&RandomRangeVector(Minimum=<-1 -1 0>, Maximum=<1 1 1>)) {isSetParameters=true}
  ]
  particleSpawnScript [
   /SetVariables(Particles.SpriteRotationRate=&RandomRangeFloat(Minimum=-1, Maximum=1, Evaluation Type=::NewEnumerator1)) {isSetParameters=true}
   &InitializeParticle(
    Color Maximum=<rgba 0.0443 0.0117 0.6875 1>
    Color Minimum=<rgba 0.7865 0.0934 0 1>
    Color Mode=::NewEnumerator2
    Lifetime Max=1
    Lifetime Min=0.3
    Lifetime Mode=::NewEnumerator1
    Sprite Rotation Mode=::NewEnumerator2
    Sprite Size Mode=::NewEnumerator3
    Uniform Sprite Size=&FloatFromCurve(
     FloatCurve=DataInterface() {propertyValues {Curve {keys [<at 0 6 RCIM_Cubic>, {arriveTangent=66.1397, interpMode=RCIM_Cubic, tangentMode=RCTM_Break, time=1, value=30}]}}}
     CurveIndex=&RandomRangeFloat(Minimum=0, Maximum=0.5)
    )
    UsePositionOffset=false
    Write Material Random=true
   )
   /Niagara/Modules/Spawn/Location/ConeLocation(
    Cone Angle:NiagaraFloat=&RandomRangeFloat(Minimum=15, Maximum=50)
    Cone Axis:Vector3f=$Emitter.ConeDir
    Cone Length:NiagaraFloat=&Multiply_Float(A=100, B=$User.GlobalScale)
    MaskSpawnGroup=false
    Module.Override Randomness=false
   )
   &AddVelocityInCone(
    Cone Axis=$Emitter.ConeDir
    Module.Use Velocity Falloff On Cone Axis=false
    Velocity Strength=&FloatFromCurve(
     FloatCurve=DataInterface() {propertyValues {Curve {keys [<at 0 10 RCIM_Cubic>, {arriveTangent=189.7314, interpMode=RCIM_Cubic, tangentMode=RCTM_Break, time=1, value=75}]}}}
     CurveIndex=&RandomRangeFloat(Minimum=0, Maximum=1)
    )
   )
   &AddVelocity(Velocity=&Multiply_VectorByFloat(
     Vector=$Output.SphereLocation.SphereVector
     Float=&RandomRangeFloat(Minimum=-0.04, Maximum=0.02, Evaluation Type=::NewEnumerator1)
    ))
  ]
  particleUpdateScript [
   &ParticleState()
   &CurlNoiseForce(Noise Frequency=10, Noise Strength=300, Pan Noise=false)
   &AccelerationForce(Acceleration=<0 0 50>)
   &InheritVelocity(Speed Limit=true, UseSpeedThreshold=false)
   &Drag(Drag=8, Use Linear Drag=true, Use Rotational Drag=false)
   &SolveForcesAndVelocity(Clamp Velocity=false, Limit Acceleration=false)
   &ScaleSpriteSize(Scale Factor=&Vector2DFromFloat(Value=&FloatFromCurve(FloatCurve=DataInterface() {propertyValues {Curve {keys [
          <at -0.0044 0.2654>
          {arriveTangent=0.7558, interpMode=RCIM_Cubic, leaveTangent=0.7558, time=0.157, value=0.5261}
          {arriveTangent=-0.6241, interpMode=RCIM_Cubic, leaveTangent=-0.6241, time=0.7034, value=0.8003}
          <at 1 0>
         ]}}})))
   &SpriteRotationRate(Rotation Rate=$Particles.SpriteRotationRate)
   &ScaleColor(
    Scale RGB=&VectorFromFloat(Value=&RandomRangeFloat(Minimum=40, Maximum=150))
    ScaleA=false
    ScaleRGB=true
   )
  ]
  rendererClasses [&NiagaraSpriteRendererProperties]
  renderers [{rendererClass=&NiagaraSpriteRendererProperties}]
 }
 Idle_FlareAndShard {
  emitterUpdateScript [
   &EmitterState(Life Cycle Mode=::NewEnumerator1, Loop Behavior=::NewEnumerator1, Loop Duration Mode=::NewEnumerator1, UseLoopDelay=false)
   &SpawnBurst_Instantaneous(Use Spawn Probability=false, UseLoopCountLimit=false)
  ]
  particleSpawnScript [
   &InitializeParticle(
    Color=<rgba 1 0.05 0.05 1>
    Lifetime=1.5
    Sprite Size=&Multiply_Vector2D_ByFloat(Float=1, Vector2D=<600 200>)
    Sprite Size Mode=::NewEnumerator4
    UsePositionOffset=false
    Write Material Random=true
   )
   /SetVariables(Particles.RoundFlareColor:LinearColor=<rgba 0.6198 0.1857 0.4055 0.5>) {isSetParameters=true}
   &DynamicMaterialParameters(
    Index 0 Param 1=&RandomRangeFloat(Minimum=0.16, Maximum=0.35, Evaluation Type=::NewEnumerator1)
    Index 0 Param 2=&RandomRangeFloat(Minimum=1, Maximum=2, Evaluation Type=::NewEnumerator1)
    Index 0 Param 3=1
    Index 0 Param 4=1
    Module.Param0WriteEnabled=true
    Module.Param1WriteEnabled=true
    Module.Param2WriteEnabled=true
    Module.Param3WriteEnabled=true
   )
   /SetVariables(
    Particles.DarkGradColor:LinearColor=<rgba 0.0833 0.0057 0.0382 0.8>
    Particles.DarkGradSize:Vector2f=&Multiply_Vector2D_ByFloat(Float=&Divide_Float(A=1, B=1.5), Vector2D=<200 200>)
   ) {isSetParameters=true}
   &InitialMeshOrientation(Use Orientation Vector=true, Use Rotation Vector=false)
  ]
  particleUpdateScript [
   &ParticleState()
   /SetVariables(Particles.UpdatedPos:Vector3f=&Add_Vector(
     A=&Multiply_VectorByFloat(
      Vector=$User.RandomPosOffset
      Float=&FloatFromCurve(
       FloatCurve=DataInterface() {propertyValues {Curve {keys [
           {arriveTangent=17.6412, interpMode=RCIM_Cubic, leaveTangent=17.6412, tangentMode=RCTM_User, time=0, value=0}
           {arriveTangent=-0.1132, interpMode=RCIM_Cubic, leaveTangent=-3.316, tangentMode=RCTM_Break, time=0.1, value=1}
           {arriveTangent=0.5686, interpMode=RCIM_Cubic, leaveTangent=0.5686, tangentMode=RCTM_User, time=0.3257, value=-0.122}
           <at 0.7 -0.05>
           {tangentMode=RCTM_User, time=1, value=0}
          ]}}}
       CurveIndex=$Emitter.NormalizedLoopAge
       Scale Curve=1.2
      )
     )
     B=/Niagara/DynamicInputs/Helpers/SimulationPosition()
    )) {isSetParameters=true}
   /SetVariables(Particles.DarkGradSize:Vector2f=&Multiply_Vector2D_ByFloat(Float=$User.GlobalScale, Vector2D=$Particles.Initial.DarkGradSize)) {isSetParameters=true}
   /SetVariables(Particles.PointLightColor:LinearColor=&MultiplyLinearColor(
     Linear Color=$Particles.Color
     Scale Factor=&MakeVector4_FromFloat(Float=&FloatFromCurve(
       FloatCurve=DataInterface() {propertyValues {Curve {keys [
           {tangentMode=RCTM_User, time=0, value=1}
           {arriveTangent=-0.0405, interpMode=RCIM_Cubic, leaveTangent=-0.0405, tangentMode=RCTM_User, time=0.5, value=1}
           {arriveTangent=-2.8606, interpMode=RCIM_Cubic, leaveTangent=-2.8606, tangentMode=RCTM_User, time=1, value=0}
          ]}}}
       Scale Curve=3
      ))
    )) {isSetParameters=true}
   /SetVariables(Particles.LightRadius=&Multiply_Float(
     A=100
     B=&Multiply_Float(
      A=$User.GlobalScale
      B=&FloatFromCurve(FloatCurve=DataInterface() {propertyValues {Curve {keys [
           {tangentMode=RCTM_User, time=0, value=1}
           {arriveTangent=-0.0405, interpMode=RCIM_Cubic, leaveTangent=-0.0405, tangentMode=RCTM_User, time=0.4, value=1.05}
           {arriveTangent=-2.3097, interpMode=RCIM_Cubic, leaveTangent=-2.3097, tangentMode=RCTM_User, time=1, value=0}
          ]}}})
     )
    )) {isSetParameters=true}
   &ScaleSpriteSize(
    Scale Sprite Size Mode=::NewEnumerator0
    Uniform Scale Factor=&Multiply_Float(A=&Multiply_Float(A=$User.GlobalScale, B=1), B=&Multiply_Float(A=1, B=2))
   )
   &ScaleColor(
    Scale Alpha=&FloatFromCurve(FloatCurve=DataInterface() {propertyValues {Curve {keys [
         {tangentMode=RCTM_User, time=0, value=1}
         {arriveTangent=-0.0405, interpMode=RCIM_Cubic, leaveTangent=-0.0405, tangentMode=RCTM_User, time=0.7009, value=1}
         {arriveTangent=-8.0089, interpMode=RCIM_Cubic, leaveTangent=-8.0089, tangentMode=RCTM_User, time=1, value=0}
        ]}}})
    Scale RGB=&VectorFromFloat(Value=&Multiply_Float(A=$User.GlobalScale, B=&Lerp_Float(A=1, B=1, Alpha=$User.CorePlusingScalar)))
    ScaleA=true
    ScaleRGB=true
   )
   &ScaleMeshSize(Scale Factor=&VectorFromFloat(Value=&Multiply_Float(A=$User.GlobalScale, B=1)))
  ]
  rendererClasses [&NiagaraLightRendererProperties]
  renderers [{rendererClass=&NiagaraLightRendererProperties, rendererIndex=0}]
 }
 Tendrils {
  emitterSpawnScript [/SetVariables() {isSetParameters=true}]
  emitterUpdateScript [
   &EmitterState(Life Cycle Mode=::NewEnumerator1, Loop Behavior=::NewEnumerator1, Loop Duration=1.5, UseLoopDelay=false)
   &SpawnRate(SpawnRate=100, Use Spawn Probability=false)
  ]
  particleSpawnScript [&InitializeRibbon(
    Color=&MakeLinearColor(
     R:NiagaraFloat=1
     G:NiagaraFloat=1
     B:NiagaraFloat=1
     A:NiagaraFloat=&FloatFromCurve(
      FloatCurve=DataInterface() {propertyValues {Curve {keys [
          {arriveTangent=-0.0565, interpMode=RCIM_Cubic, leaveTangent=-0.0565, tangentMode=RCTM_User, time=0, value=0}
          {arriveTangent=-0.1132, interpMode=RCIM_Cubic, leaveTangent=-0.171, tangentMode=RCTM_Break, time=0.03, value=1}
          {arriveTangent=-0.4165, interpMode=RCIM_Cubic, leaveTangent=-0.4165, tangentMode=RCTM_User, time=0.9, value=0.95}
          {tangentMode=RCTM_User, time=1, value=0}
         ]}}}
      CurveIndex=$Emitter.NormalizedLoopAge
      Scale Curve=1.2
     )
    )
    Lifetime=0.4
    Position Offset:Vector3f=&Multiply_VectorByFloat(
     Vector=$User.RandomPosOffset
     Float=&FloatFromCurve(
      FloatCurve=DataInterface() {propertyValues {Curve {keys [
          {arriveTangent=14.0924, interpMode=RCIM_Cubic, leaveTangent=14.0924, tangentMode=RCTM_User, time=0, value=0}
          {arriveTangent=-0.1132, interpMode=RCIM_Cubic, leaveTangent=-3.316, tangentMode=RCTM_Break, time=0.1, value=1}
          {arriveTangent=0.5686, interpMode=RCIM_Cubic, leaveTangent=0.5686, tangentMode=RCTM_User, time=0.3257, value=-0.122}
          <at 0.7 -0.05>
          {tangentMode=RCTM_User, time=1, value=0}
         ]}}}
      CurveIndex=$Emitter.NormalizedLoopAge
      Scale Curve=1.2
     )
    )
    Ribbon Width=&FloatFromCurve(
     FloatCurve=DataInterface() {propertyValues {Curve {keys [
         {interpMode=RCIM_Cubic, tangentMode=RCTM_User, time=0, value=0.5}
         <at 0.1 1.0125>
         {arriveTangent=-0.4305, interpMode=RCIM_Cubic, leaveTangent=-0.4305, tangentMode=RCTM_User, time=0.5, value=0.9236}
         {tangentMode=RCTM_User, time=1, value=0.75}
        ]}}}
     CurveIndex=$Emitter.NormalizedLoopAge
     Scale Curve=60
    )
    UsePositionOffset=true
    Write Material Random=true
   )]
  particleUpdateScript [
   &ParticleState()
   &CurlNoiseForce(Noise Frequency=5, Noise Strength=250, Pan Noise=false)
   &SolveForcesAndVelocity(Clamp Velocity=false, Limit Acceleration=false)
   &ScaleRibbonWidth(Ribbon Width Scale=&FloatFromCurve(FloatCurve=DataInterface() {propertyValues {Curve {keys [
         {arriveTangent=6.95, interpMode=RCIM_Cubic, leaveTangent=6.95, tangentMode=RCTM_User, time=0, value=0}
         {arriveTangent=2, interpMode=RCIM_Cubic, leaveTangent=2, time=0.1, value=0.7}
         {interpMode=RCIM_Cubic, tangentMode=RCTM_User, time=0.5, value=1}
         {arriveTangent=-0.0784, interpMode=RCIM_Cubic, leaveTangent=-0.0784, tangentMode=RCTM_User, time=1, value=0}
        ]}}}))
   /SetVariables(Particles.RibbonWidthB:NiagaraFloat=&FloatFromCurve(
     FloatCurve=DataInterface() {propertyValues {Curve {keys [
         {arriveTangent=6.95, interpMode=RCIM_Cubic, leaveTangent=6.95, tangentMode=RCTM_User, time=0, value=0}
         {arriveTangent=2, interpMode=RCIM_Cubic, leaveTangent=2, time=0.1, value=0.7}
         {interpMode=RCIM_Cubic, tangentMode=RCTM_User, time=0.5, value=1}
         {arriveTangent=-0.0784, interpMode=RCIM_Cubic, leaveTangent=-0.0784, tangentMode=RCTM_User, time=1, value=0.5}
        ]}}}
     Scale Curve=80
    )) {isSetParameters=true}
   &ScaleColor(
    Scale Alpha=&FloatFromCurve(FloatCurve=DataInterface() {propertyValues {Curve {keys [
         {tangentMode=RCTM_User, time=0, value=1}
         <at 0.02 1>
         {arriveTangent=-0.0405, interpMode=RCIM_Cubic, leaveTangent=-0.0405, tangentMode=RCTM_User, time=0.7, value=1}
         {arriveTangent=-5.6161, interpMode=RCIM_Cubic, leaveTangent=-5.6161, tangentMode=RCTM_User, time=1, value=0}
        ]}}})
    ScaleA=true
    ScaleRGB=true
   )
  ]
  rendererClasses [&NiagaraRibbonRendererProperties]
  renderers [{rendererClass=&NiagaraRibbonRendererProperties}]
 }
}
summary {
 systemName=NS_Caretaker_CoreGrabber
 emitters [
  {emitterName=Embers_Dark, simTarget=GPUComputeSim, rendererClasses [&NiagaraSpriteRendererProperties]}
  {emitterName=Tendrils, rendererClasses [&NiagaraRibbonRendererProperties]}
  {emitterName=Idle_FlareAndShard, rendererClasses [&NiagaraLightRendererProperties]}
 ]
 userVariables [
  User.CorePlusingScalar {type=NiagaraFloat, defaultValue=0, description=""}
  User.GlobalScale {type=NiagaraFloat, defaultValue=1, description=""}
  User.InWater {type=NiagaraBool}
  User.Open {type=NiagaraFloat, defaultValue=1, description=""}
  User.RandomPosOffset {type=Vector3f, description="", defaultValue {struct=Vector3f, value=<0 0 0>}}
  User.WaveLength {type=NiagaraFloat, defaultValue=0, description=""}
  User.WaveScale {type=NiagaraFloat, defaultValue=0, description=""}
 ]
}
